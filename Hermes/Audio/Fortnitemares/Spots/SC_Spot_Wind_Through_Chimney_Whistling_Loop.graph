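# Sound cue for a looping whistling wind-through-chimney ambient spot sound in Fortnitemares, routed to the indoor ambient one-shots sound class.
# A looping wave feeds an indefinitely looping Enveloper that gently dips and restores volume, then a Modulator that slightly randomizes pitch.
# It borrows a desert wasteland window-wind attenuation asset and references a one-voice prop concurrency set that stops the lowest-priority sound.
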SoundCue /Hermes/Audio/Fortnitemares/Spots/SC_Spot_Wind_Through_Chimney_Whistling_Loop
root=Modulator_3
nodes [
 {
  id=Enveloper_0
  children [WavePlayer_0]
  properties {
   bLoop=true
   bLoopIndefinitely=true
   loopEnd=25
   pitchCurve {editorCurveData {keys [{time=0}]}}
   volumeCurve {editorCurveData {keys [<at 0 1>,<at 12.5 0.8>,<at 25 1>]}}
  }
 }
 {id=Modulator_3, children [Enveloper_0], properties {pitchMax=1.1, pitchMin=0.9, volumeMax=1, volumeMin=1}}
 {id=WavePlayer_0, properties {bLooping=true, soundWaveAssetPtr=&/Spot_Wind_Through_Chimney_Whistling_Loop}}
]
properties {
 attenuationSettings=/Game/Sounds/Helios/Biomes/DesertWasteland/Spots/Spot_Window_Wind_Sphere_Large_Att
 bEnableBusSends=false
 maxDistance=1450
 soundClassObject=/Game/Sounds/FortSoundClassesAndModes/Fort_Ambient_OneShots_Indoors
 volumeMultiplier=0.7
 concurrencyOverrides {maxCount=3, retriggerTime=2, voiceStealReleaseTime=2, volumeScaleAttackTime=1}
 concurrencySet [/AmbientAudioBase/Concurrency/AmbientConc_1Max_Prop_StopLowestPriority_Concurrency]
 preEffectBusSends [{}]
}
