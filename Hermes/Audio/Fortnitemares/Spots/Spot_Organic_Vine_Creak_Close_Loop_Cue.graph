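# Looping Fortnitemares ambient spot cue for close organic vine creaks, mixing two randomized layers with distant-creak attenuation.
# One layer loops pumpkin vine growth movement one-shots back to back, picked through nested randoms with pitch and volume modulation.
# The other loops close creak one-shots after random delays, through a quality-level node whose branches all share one modulator.
# A concurrency override caps simultaneous instances, and the cue uses the indoor ambient one-shots sound class.
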SoundCue /Hermes/Audio/Fortnitemares/Spots/Spot_Organic_Vine_Creak_Close_Loop_Cue
root=Mixer_1
nodes [
 {id=Delay_0, children [QualityLevel_0], properties {delayMax=22, delayMin=7}}
 {id=Looping_0, children [Delay_0]}
 {id=Looping_1, children [Modulator_0]}
 {id=Mixer_1, children [Looping_0,Looping_1], properties {inputVolume [0.75,0.65]}}
 {id=Modulator_0, children [Random_5], properties {pitchMax=1.1, pitchMin=0.7, volumeMax=1, volumeMin=0.5}}
 {id=Modulator_3, children [Random_3], properties {pitchMax=1.1, pitchMin=0.9, volumeMin=0.5}}
 {id=QualityLevel_0, children [Modulator_3,Modulator_3,Modulator_3,Modulator_3]}
 {id=Random_0, children [WavePlayer_0,WavePlayer_1,WavePlayer_2,WavePlayer_3,WavePlayer_4], properties {weights [1,1,1,1,1]}}
 {id=Random_1, children [WavePlayer_5,WavePlayer_9], properties {weights [1,1]}}
 {id=Random_2, children [WavePlayer_10, WavePlayer_11], properties {weights [1,1]}}
 {id=Random_3, children [Random_1,Random_2], properties {weights [1,1]}}
 {
  id=Random_4
  children [WavePlayer_6, WavePlayer_7, WavePlayer_8, WavePlayer_12, WavePlayer_13, WavePlayer_14]
  properties {weights [1,1,1,1,1,1]}
 }
 {id=Random_5, children [Random_0,Random_4], properties {weights [1,1]}}
 {id=WavePlayer_0, properties {soundWaveAssetPtr=/Hermes/Audio/Fortnitemares/Spots/Oneshots/Spot_Pumpkin_Vine_Growth_Movement_01}}
 {id=WavePlayer_1, properties {soundWaveAssetPtr=/Hermes/Audio/Fortnitemares/Spots/Oneshots/Spot_Pumpkin_Vine_Growth_Movement_02}}
 {id=WavePlayer_10, properties {soundWaveAssetPtr=/Hermes/Audio/Fortnitemares/Spots/Oneshots/Spot_Organic_Vine_Creak_Close_03}}
 {id=WavePlayer_11, properties {soundWaveAssetPtr=/Hermes/Audio/Fortnitemares/Spots/Oneshots/Spot_Organic_Vine_Creak_Close_04}}
 {id=WavePlayer_12, properties {soundWaveAssetPtr=/Hermes/Audio/Fortnitemares/Spots/Oneshots/Spot_Pumpkin_Vine_Growth_Movement_09}}
 {id=WavePlayer_13, properties {soundWaveAssetPtr=/Hermes/Audio/Fortnitemares/Spots/Oneshots/Spot_Pumpkin_Vine_Growth_Movement_10}}
 {id=WavePlayer_14, properties {soundWaveAssetPtr=/Hermes/Audio/Fortnitemares/Spots/Oneshots/Spot_Pumpkin_Vine_Growth_Movement_11}}
 {id=WavePlayer_2, properties {soundWaveAssetPtr=/Hermes/Audio/Fortnitemares/Spots/Oneshots/Spot_Pumpkin_Vine_Growth_Movement_03}}
 {id=WavePlayer_3, properties {soundWaveAssetPtr=/Hermes/Audio/Fortnitemares/Spots/Oneshots/Spot_Pumpkin_Vine_Growth_Movement_04}}
 {id=WavePlayer_4, properties {soundWaveAssetPtr=/Hermes/Audio/Fortnitemares/Spots/Oneshots/Spot_Pumpkin_Vine_Growth_Movement_05}}
 {id=WavePlayer_5, properties {soundWaveAssetPtr=/Hermes/Audio/Fortnitemares/Spots/Oneshots/Spot_Organic_Vine_Creak_Close_01}}
 {id=WavePlayer_6, properties {soundWaveAssetPtr=/Hermes/Audio/Fortnitemares/Spots/Oneshots/Spot_Pumpkin_Vine_Growth_Movement_06}}
 {id=WavePlayer_7, properties {soundWaveAssetPtr=/Hermes/Audio/Fortnitemares/Spots/Oneshots/Spot_Pumpkin_Vine_Growth_Movement_07}}
 {id=WavePlayer_8, properties {soundWaveAssetPtr=/Hermes/Audio/Fortnitemares/Spots/Oneshots/Spot_Pumpkin_Vine_Growth_Movement_08}}
 {id=WavePlayer_9, properties {soundWaveAssetPtr=/Hermes/Audio/Fortnitemares/Spots/Oneshots/Spot_Organic_Vine_Creak_Close_02}}
]
properties {
 attenuationSettings=/Game/Sounds/Attenuation/Fort_TileAmbient_OneShot_Distant_Creak_Linear_Att
 bEnableBusSends=false
 bOverrideConcurrency=true
 maxDistance=8700
 soundClassObject=/Game/Sounds/FortSoundClassesAndModes/Fort_Ambient_OneShots_Indoors
 volumeMultiplier=0.5
 concurrencyOverrides {maxCount=2, voiceStealReleaseTime=0.25}
 preEffectBusSends [{}]
}
